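# Base material for the SaltyBrim Pin backpack, layering an animated side-scrolling portal onto the 2023 anime character shader.
# Background, middle and foreground textures pan at separate speeds with bump-offset parallax, then composite inside a UV1 sphere mask.
# A noise-distorted edge plus Fresnel adds the portal tint, and scanlines are generated as an overlay.
# The effect is built separately for lit and shaded colors between the anime shader's Pre and Post functions.
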Material /BRCosmetics/Accessories/FORT_Backpacks/Backpack_M_MED_SaltyBrim_Pin/Materials/M_SaltyBrim_Pin_Backpack_Base
graph {
 properties {
  bAutomaticallySetUsageInEditor=false
  bUseMaterialAttributes=true
  bUsedWithSkeletalMesh=true
  pixelDepthOffsetMode=PDOM_Legacy
  previewMesh=/BRCosmetics/Characters/Player/Male/Medium/Heads/M_MED_ScribeSpider_Head/Meshes/M_MED_ScribeSpider_Head
  shadingModel=MSM_Cloth
  usedShadingModels=Cloth
 }
 nodes [
  n1 = MaterialFunctionCall!(CustomUV) {materialFunction=/Game/Global/Materials/Character_Parent_2023/Anime_Parent_2023/MaterialFunctions/MFC_AnimeShader_2023_Pre}
  n2 = MaterialFunctionCall!(CustomLightVector, OverrideNormal_WS, STT=@n1.STT, Normal_Map=@n1.STT, DFL=@n1.STT, FX=@n1.STT) {materialFunction=/Game/Global/Materials/Character_Parent_2023/Anime_Parent_2023/MaterialFunctions/MFC_AnimeShader_2023_Calculations}
  MaterialFunctionCall!(
   Shading_Gradient=@n2.Shading_Gradient
   Threshold_Gradient=@n2.Shading_Gradient
   Specular_Gradient=@n2.Shading_Gradient
   Sheen_Mask=@n2.Shading_Gradient
   Thickness=@n2.Shading_Gradient
   DFL_Mask=@n2.Shading_Gradient
   FX=@n2.Shading_Gradient
   Metallic_Gradient=@n2.Shading_Gradient
   RimLight_Gradient=@n2.Shading_Gradient
   NormalBuffer=@n2.Shading_Gradient
   --------
   CINE_LightSources=@n2.Shading_Gradient
   CINE_LightColor=@n2.Shading_Gradient
   ---------
   Color_Lit=NamedRerouteUsage()
   Color_Shaded=NamedRerouteUsage()
   SRM=@n1.SRM
   Emissive=@n1.SRM
   DFL_Color=@n1.SRM
   Specular_Color=@n1.SRM
   RimLight_Color=@n1.SRM
  ) {materialFunction=/Game/Global/Materials/Character_Parent_2023/Anime_Parent_2023/MaterialFunctions/MFC_AnimeShader_2023_Post}
  NamedRerouteDeclaration(None=@n1.Color_Lit) {name=IN_ColorLit, nodeColor=<rgba 0.5703 0 1 1>}
  NamedRerouteDeclaration(None=@n1.Color_Shaded) {name=IN_ColorShaded, nodeColor=<rgba 0 1 0.25 1>}
  n3 = ScalarParameter() {defaultValue=0.0125, group=SaltyBrim, parameterName=BackgroundSpeed}
  n4 = Constant()
  n5 = MaterialFunctionCall!(
   UV Coord=Panner(Coordinate=NamedRerouteUsage(), Speed=AppendVector(A=@n3, B=@n4))
   Height=Constant() {r=1}
   ReferencePlane=Constant()
   HeightRatio=ScalarParameter() {defaultValue=0.3, group=SaltyBrim, parameterName=BackgroundHeight}
   Normal=Constant3Vector() {constant=<rgba 1 0 0 0>}
  ) {materialFunction=/Game/Global/MaterialFunctions/MF_BumpOffset_Accurate}
  n6 = NamedRerouteUsage()
  n7 = AppendVector(A=ScalarParameter() {defaultValue=0.025, group=SaltyBrim, parameterName=MiddleSpeed}, B=Constant())
  n8 = TextureSample(UVs=MaterialFunctionCall(
    UV Coord=Panner(Coordinate=@n6, Speed=@n7)
    Height=Constant() {r=1}
    ReferencePlane=Constant()
    HeightRatio=ScalarParameter() {defaultValue=0.15, group=SaltyBrim, parameterName=MiddleHeight}
    Normal=Constant3Vector() {constant=<rgba 1 0 0 0>}
   ) {materialFunction=&MF_BumpOffset_Accurate}.Result) {texture=/BRCosmetics/Accessories/FORT_Backpacks/Backpack_M_MED_SaltyBrim_Pin/Textures/T_SaltyBrim_Pin_SideScroll_MG}
  n9 = NamedRerouteUsage()
  n10 = AppendVector(A=ScalarParameter() {defaultValue=0.05, group=SaltyBrim, parameterName=ForegroundSpeed}, B=Constant())
  n11 = TextureSample(UVs=MaterialFunctionCall(
    UV Coord=Panner(Coordinate=@n9, Speed=@n10)
    Height=Constant() {r=1}
    ReferencePlane=Constant()
    HeightRatio=ScalarParameter() {defaultValue=0.05, group=SaltyBrim, parameterName=ForegroundHeight}
    Normal=Constant3Vector() {constant=<rgba 1 0 0 0>}
   ) {materialFunction=&MF_BumpOffset_Accurate}.Result) {texture=/BRCosmetics/Accessories/FORT_Backpacks/Backpack_M_MED_SaltyBrim_Pin/Textures/T_SaltyBrim_Pin_SideScroll_FG}
  n12 = LinearInterpolate(
   A=LinearInterpolate(
    A=TextureSample(UVs=@n5.Result) {texture=/BRCosmetics/Accessories/FORT_Backpacks/Backpack_M_MED_SaltyBrim_Pin/Textures/T_SaltyBrim_Pin_SideScroll_BG}.RGB
    B=@n8.RGB
    Alpha=@n8.RGB
   )
   B=@n11.RGB
   Alpha=@n11.RGB
  )
  n13 = NamedRerouteDeclaration(None=SphereMask(A=TextureCoordinate() {coordinateIndex=1}, B=Constant() {r=0.5}, Radius=Constant() {r=0.5})) {name=PortalMask, nodeColor=<rgba 1 0 0.0234 1>}
  n14 = TextureCoordinate() {coordinateIndex=1}
  n15 = TextureSample(UVs=Panner(Coordinate=TextureCoordinate() {coordinateIndex=1}) {speedX=0.25, speedY=0.13}) {texture=/Game/Global/Textures/Default/Noise/T_SoftNoise}
  n16 = TextureSample(UVs=Panner(Coordinate=TextureCoordinate() {coordinateIndex=1, uTiling=0.5, vTiling=0.5}) {speedX=-0.12, speedY=-0.1}) {texture=&T_SoftNoise}
  n17 = Multiply(A=ComponentMask(None=Multiply(A=@n15.RGB, B=@n16.RGB)) {g=true, r=true}) {constB=0.1}
  n18 = OneMinus(None=SphereMask(
    A=Add(A=@n14, B=@n17)
    B=Constant() {r=0.5}
    Radius=Constant() {r=0.5}
    Hardness=ScalarParameter() {defaultValue=0.5, group=SaltyBrim, parameterName=PortalBorderHardness}
   ) {hardnessPercent=50})
  n19 = Fresnel(
   ExponentIn=ScalarParameter() {defaultValue=2, group=SaltyBrim, parameterName=PortalTintFresnelExp}
   BaseReflectFractionIn=ScalarParameter() {defaultValue=0.1, group=SaltyBrim, parameterName=PortalTintFresnelBase}
  ) {baseReflectFraction=0.1, exponent=2}
  n20 = Multiply(A=Add(A=SmoothStep(Value=@n18), B=SmoothStep(Value=@n19) {constMax=1.5, constMin=-0.1}), B=NamedRerouteUsage())
  n21 = LinearInterpolate(A=VectorParameter() {group=SaltyBrim, parameterName=PortalTintColor, defaultValue=<rgba 0 1 2 1>}.RGB, B=Constant() {r=1}, Alpha=@n20)
  n22 = LinearInterpolate(A=LinearInterpolate(A=NamedRerouteUsage(), B=@n12, Alpha=@n13), B=@n21, Alpha=@n20)
  n23 = NamedRerouteUsage()
  NamedRerouteDeclaration(None=LinearInterpolate(A=@n22, B=Multiply(A=@n22, B=@n23), Alpha=NamedRerouteUsage())) {name=OUT_ColorLit, nodeColor=<rgba 1 0 0.0703 1>}
  n24 = TextureCoordinate() {coordinateIndex=1}
  NamedRerouteDeclaration(None=LinearInterpolate(
    B=MaterialFunctionCall!(ScanLine_Tiling_1, Scanline_PannerSpeed_1, ScanLine_Tiling, Scanline_PannerSpeed, UVs=@n24) {materialFunction=/Game/Weapons/WeaponSkins/MaterialFunction/ScanLines}.Result
    Alpha=ScalarParameter() {defaultValue=0.5, group=Scanlines, parameterName=Scanline_Opacity, sortPriority=50}
   ) {constA=1}) {name=Scanline, nodeColor=<rgba 1 0.2109 0 1>}
  n25 = LinearInterpolate(A=LinearInterpolate(A=NamedRerouteUsage(), B=@n12, Alpha=@n13), B=@n21, Alpha=@n20)
  n26 = NamedRerouteUsage()
  NamedRerouteDeclaration(None=LinearInterpolate(A=@n25, B=Multiply(A=@n25, B=@n26), Alpha=NamedRerouteUsage())) {name=OUT_ColorShaded, nodeColor=<rgba 0 0.3906 1 1>}
  n27 = VectorParameter() {group=SaltyBrim, parameterName=UV Tiling, defaultValue=<rgba 1 1 0 0>}
  NamedRerouteDeclaration(None=Multiply(A=TextureCoordinate() {coordinateIndex=1}, B=ComponentMask(None=@n27.RGB) {g=true, r=true})) {name=TiledUVs, nodeColor=<rgba 0.1484 1 0 1>}
 ]
}
